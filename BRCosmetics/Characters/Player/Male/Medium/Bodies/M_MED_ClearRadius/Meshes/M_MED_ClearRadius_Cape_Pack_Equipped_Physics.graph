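# Physics asset for the equipped cape pack on the male medium ClearRadius character body, defining cape bodies and body colliders.
# Four capsule chains hang from two clavicle anchors and a shared cape root, with tight swing and twist limits.
# Angularly free constraints tie adjacent chains at their third segments, linking the strips laterally.
# Pelvis, spine, thigh and right-arm capsules have no constraints, and the left arm has no body.
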PhysicsAsset /BRCosmetics/Characters/Player/Male/Medium/Bodies/M_MED_ClearRadius/Meshes/M_MED_ClearRadius_Cape_Pack_Equipped_Physics
bodies {
 cape_01_l [{shapeType=Capsule, length=11.0049, radius=3.9367, shapeName=cape_mid_04_l_capsule, center=<9.1011 0 0>, rotation=<rot 90 0 0>}]
 cape_01_r [{shapeType=Capsule, length=10.3742, radius=4.5674, shapeName=cape_mid_04_l_capsule, center=<9.1011 0 0>, rotation=<rot 90 0 0>}]
 cape_02_l [{shapeType=Capsule, length=11.0049, radius=3.9367, shapeName=cape_mid_04_l_capsule, center=<9.1011 0 0>, rotation=<rot 90 0 0>}]
 cape_02_r [{shapeType=Capsule, length=9.6832, radius=5.2584, shapeName=cape_mid_04_l_capsule, center=<9.1011 0 0>, rotation=<rot 90 0 0>}]
 cape_03_l [{shapeType=Capsule, length=11.0049, radius=3.9367, shapeName=cape_mid_04_l_capsule, center=<9.1011 0 0>, rotation=<rot 90 0 0>}]
 cape_03_r [{shapeType=Capsule, length=9.6832, radius=5.2584, shapeName=cape_mid_04_l_capsule, center=<9.1011 0 0>, rotation=<rot 90 0 0>}]
 cape_04_l [{shapeType=Capsule, length=11.0049, radius=3.9367, shapeName=cape_mid_04_l_capsule, center=<9.1011 0 0>, rotation=<rot 90 0 0>}]
 cape_04_r [{shapeType=Capsule, length=9.6832, radius=5.2584, shapeName=cape_mid_04_l_capsule, center=<9.1011 0 0>, rotation=<rot 90 0 0>}]
 cape_mid_01_l [{shapeType=Capsule, length=11.0049, radius=3.9367, shapeName=cape_mid_04_l_capsule, center=<9.1011 0 0>, rotation=<rot 90 0 0>}]
 cape_mid_01_r [{shapeType=Capsule, length=11.0049, radius=3.9367, shapeName=cape_mid_04_l_capsule, center=<9.1011 0 0>, rotation=<rot 90 0 0>}]
 cape_mid_02_l [{shapeType=Capsule, length=11.0049, radius=3.9367, shapeName=cape_mid_04_l_capsule, center=<9.1011 0 0>, rotation=<rot 90 0 0>}]
 cape_mid_02_r [{shapeType=Capsule, length=11.0049, radius=3.9367, shapeName=cape_mid_04_l_capsule, center=<9.1011 0 0>, rotation=<rot 90 0 0>}]
 cape_mid_03_l [{shapeType=Capsule, length=11.0049, radius=3.9367, shapeName=cape_mid_04_l_capsule, center=<9.1011 0 0>, rotation=<rot 90 0 0>}]
 cape_mid_03_r [{shapeType=Capsule, length=11.0049, radius=3.9367, shapeName=cape_mid_04_l_capsule, center=<9.1011 0 0>, rotation=<rot 90 0 0>}]
 cape_mid_04_l [{shapeType=Capsule, length=11.0049, radius=3.9367, shapeName=cape_mid_04_l_capsule, center=<9.1011 0 0>, rotation=<rot 90 0 0>}]
 cape_mid_04_r [{shapeType=Capsule, length=11.0049, radius=3.9367, shapeName=cape_mid_04_l_capsule, center=<9.1011 0 0>, rotation=<rot 90 0 0>}]
 cape_root [{shapeType=Sphere, radius=0.5, shapeName=cape_root_sphere}]
 clavicle_cape_l [{shapeType=Sphere, radius=0.5, shapeName=clavicle_cape_l_sphere}]
 clavicle_cape_r [{shapeType=Sphere, radius=0.5, shapeName=clavicle_cape_r_sphere}]
 lowerarm_r [{
   shapeType=Capsule
   length=25.0018
   radius=8.8941
   shapeName=lowerarm_r_capsule
   center=<-9.9731 -0.8184 0>
   rotation=<rot 90 -173.7581 -170.188>
  }]
 pelvis [{shapeType=Capsule, length=8.2808, radius=11.3944, shapeName=pelvis_capsule, center=<-5.6536 -1.5263 0>}]
 spine_03 [{
   shapeType=Capsule
   length=13.8408
   radius=11.7577
   shapeName=spine_03_capsule
   center=<-6.161 -0.4294 0>
   rotation=<rot 90 -179.1048 175.6438>
  }]
 spine_05 [{shapeType=Capsule, length=14.1819, radius=11.1039, shapeName=spine_05_capsule, center=<-7.6739 1.7776 0>, rotation=<rot 90 0 -7.7769>}]
 thigh_l [{
   shapeType=Capsule
   length=22.0841
   radius=7.9159
   shapeName=thigh_l_capsule
   center=<-18.1221 -2.1147 0.4051>
   rotation=<rot 88.1093 179.9995 -175.1682>
  }]
 thigh_r [{
   shapeType=Capsule
   length=22.0477
   radius=7.9523
   shapeName=thigh_r_capsule
   center=<17.9586 1.7377 -0.711>
   rotation=<rot 86.5772 179.9992 -175.9254>
  }]
 upperarm_r [{
   shapeType=Capsule
   length=23.4833
   radius=6.5167
   shapeName=upperarm_r_capsule
   center=<-11.6375 0 0.4117>
   rotation=<rot 89.3539 -179.9914 -179.9915>
  }]
}
constraints [
 {bone1Name=cape_01_r, bone2Name=clavicle_cape_r, swing1LimitDegrees=5, swing2LimitDegrees=1, twistLimitDegrees=3}
 {bone1Name=cape_02_r, bone2Name=cape_01_r, swing1LimitDegrees=8, swing2LimitDegrees=1, twistLimitDegrees=3}
 {bone1Name=cape_03_r, bone2Name=cape_02_r, swing1LimitDegrees=8, swing2LimitDegrees=1, twistLimitDegrees=3}
 {bone1Name=cape_04_r, bone2Name=cape_03_r, swing1LimitDegrees=5, swing2LimitDegrees=3, twistLimitDegrees=5}
 {bone1Name=cape_01_l, bone2Name=clavicle_cape_l, swing1LimitDegrees=5, swing2LimitDegrees=5, twistLimitDegrees=3}
 {bone1Name=cape_02_l, bone2Name=cape_01_l, swing1LimitDegrees=8, swing2LimitDegrees=3, twistLimitDegrees=3}
 {bone1Name=cape_03_l, bone2Name=cape_02_l, swing1LimitDegrees=8, swing2LimitDegrees=3, twistLimitDegrees=5}
 {bone1Name=cape_04_l, bone2Name=cape_03_l, swing1LimitDegrees=5, swing2LimitDegrees=5, twistLimitDegrees=5}
 {bone1Name=cape_mid_04_r, bone2Name=cape_mid_03_r, swing1LimitDegrees=5, swing2LimitDegrees=3, twistLimitDegrees=5}
 {bone1Name=cape_mid_03_r, bone2Name=cape_mid_02_r, swing1LimitDegrees=8, swing2LimitDegrees=1, twistLimitDegrees=3}
 {bone1Name=cape_mid_02_r, bone2Name=cape_mid_01_r, swing1LimitDegrees=8, swing2LimitDegrees=1, twistLimitDegrees=3}
 {bone1Name=cape_mid_01_r, bone2Name=cape_root, swing1LimitDegrees=5, swing2LimitDegrees=1, twistLimitDegrees=3}
 {bone1Name=cape_mid_04_l, bone2Name=cape_mid_03_l, swing1LimitDegrees=5, swing2LimitDegrees=5, twistLimitDegrees=5}
 {bone1Name=cape_mid_03_l, bone2Name=cape_mid_02_l, swing1LimitDegrees=8, swing2LimitDegrees=3, twistLimitDegrees=5}
 {bone1Name=cape_mid_02_l, bone2Name=cape_mid_01_l, swing1LimitDegrees=8, swing2LimitDegrees=3, twistLimitDegrees=3}
 {bone1Name=cape_mid_01_l, bone2Name=cape_root, swing1LimitDegrees=5, swing2LimitDegrees=5, twistLimitDegrees=3}
 {
  bone1Name=cape_03_r
  bone2Name=cape_mid_03_r
  swing1LimitDegrees=45
  swing1Motion=Free
  swing2LimitDegrees=45
  swing2Motion=Free
  twistLimitDegrees=45
  twistMotion=Free
 }
 {
  bone1Name=cape_mid_03_l
  bone2Name=cape_03_l
  swing1LimitDegrees=45
  swing1Motion=Free
  swing2LimitDegrees=45
  swing2Motion=Free
  twistLimitDegrees=45
  twistMotion=Free
 }
 {
  bone1Name=cape_mid_03_r
  bone2Name=cape_mid_03_l
  swing1LimitDegrees=45
  swing1Motion=Free
  swing2LimitDegrees=45
  swing2Motion=Free
  twistLimitDegrees=45
  twistMotion=Free
 }
]
