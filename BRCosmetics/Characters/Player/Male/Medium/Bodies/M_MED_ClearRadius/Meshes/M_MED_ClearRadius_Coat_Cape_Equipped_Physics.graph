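# Physics asset for the equipped coat cape of the male medium ClearRadius player cosmetic, defining bodies for its dynamic coat bones.
# Five chains (back, front left/right, front side left/right) each link a small base sphere to parent and tip bodies, mostly capsules.
# Constraints use tight swing and twist limits, with the back chain allowing wider second-axis swing.
# Two large, unconstrained spheres on spine_03 and spine_04 provide upper-torso collision shapes.
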PhysicsAsset /BRCosmetics/Characters/Player/Male/Medium/Bodies/M_MED_ClearRadius/Meshes/M_MED_ClearRadius_Coat_Cape_Equipped_Physics
bodies {
 dyn_coat_bk_01 [{shapeType=Capsule, length=6.8617, radius=1.2072, shapeName=dyn_coat_frnt_l_01_capsule, center=<1.6572 0 0>, rotation=<rot 0 0 90>}]
 dyn_coat_bk_base [{shapeType=Sphere, radius=0.5, shapeName=dyn_coat_bk_base_sphere}]
 dyn_coat_bk_parent [{shapeType=Capsule, length=6.8617, radius=1.2072, shapeName=dyn_coat_frnt_l_01_capsule, center=<1.6572 0 0>, rotation=<rot 0 0 90>}]
 dyn_coat_frnt_l_01 [{shapeType=Capsule, length=6.6092, radius=1.4598, shapeName=dyn_coat_frnt_l_01_capsule, center=<1.6572 0 0>, rotation=<rot 0 0 79.784>}]
 dyn_coat_frnt_l_base [{shapeType=Sphere, radius=0.5, shapeName=dyn_coat_frnt_l_base_sphere}]
 dyn_coat_frnt_l_parent [
  {shapeType=Sphere, radius=1.4598, shapeName=dyn_coat_frnt_l_01_capsule, center=<1.6572 0 0>}
  {shapeType=Capsule, length=5.5891, radius=2.4798, shapeName=dyn_coat_frnt_l_01_capsule, center=<1.6572 0 0>, rotation=<rot 0 0 79.784>}
 ]
 dyn_coat_frnt_r_01 [{
   shapeType=Capsule
   length=10.371
   radius=2.5898
   shapeName=dyn_coat_frnt_l_01_capsule
   center=<3.09 0.0001 -0.0001>
   rotation=<rot -0.006 0.0044 99.1407>
  }]
 dyn_coat_frnt_r_base [{shapeType=Sphere, radius=0.5, shapeName=dyn_coat_frnt_r_base_sphere}]
 dyn_coat_frnt_r_parent [{
   shapeType=Capsule
   length=10.371
   radius=2.5898
   shapeName=dyn_coat_frnt_l_01_capsule
   center=<3.09 0.0001 -0.0001>
   rotation=<rot -0.006 0.0044 99.1407>
  }]
 dyn_coat_frnt_side_l_01 [{
   shapeType=Capsule
   length=5.5891
   radius=2.4798
   shapeName=dyn_coat_frnt_l_01_capsule
   center=<1.6572 0 0>
   rotation=<rot 0 -0.0001 103.3384>
  }]
 dyn_coat_frnt_side_l_base [{shapeType=Sphere, radius=0.5, shapeName=dyn_coat_frnt_side_l_base_sphere}]
 dyn_coat_frnt_side_l_parent [{
   shapeType=Capsule
   length=5.5891
   radius=2.4798
   shapeName=dyn_coat_frnt_l_01_capsule
   center=<1.6572 0 0>
   rotation=<rot 0 -0.0001 103.3384>
  }]
 dyn_coat_frnt_side_r_01 [{shapeType=Capsule, length=5.5891, radius=2.4798, shapeName=dyn_coat_frnt_l_01_capsule, center=<1.6572 0 0>, rotation=<rot 0 0 96.9771>}]
 dyn_coat_frnt_side_r_base [{shapeType=Sphere, radius=0.5, shapeName=dyn_coat_frnt_side_r_base_sphere}]
 dyn_coat_frnt_side_r_parent [{shapeType=Capsule, length=5.5891, radius=2.4798, shapeName=dyn_coat_frnt_l_01_capsule, center=<1.6572 0 0>, rotation=<rot 0 0 96.9771>}]
 spine_03 [{shapeType=Sphere, radius=10.4608, shapeName=spine_04_sphere, center=<0.6653 1.035 0>}]
 spine_04 [{shapeType=Sphere, radius=10.4776, shapeName=spine_04_sphere, center=<-0.4708 2.0901 0>}]
}
constraints [
 {bone1Name=dyn_coat_frnt_side_r_01, bone2Name=dyn_coat_frnt_side_r_parent, swing1LimitDegrees=10, swing2LimitDegrees=10, twistLimitDegrees=3}
 {
  bone1Name=dyn_coat_frnt_side_r_parent
  bone2Name=dyn_coat_frnt_side_r_base
  swing1LimitDegrees=10
  swing2LimitDegrees=10
  twistLimitDegrees=3
 }
 {bone1Name=dyn_coat_frnt_side_l_01, bone2Name=dyn_coat_frnt_side_l_parent, swing1LimitDegrees=10, swing2LimitDegrees=10, twistLimitDegrees=3}
 {
  bone1Name=dyn_coat_frnt_side_l_parent
  bone2Name=dyn_coat_frnt_side_l_base
  swing1LimitDegrees=10
  swing2LimitDegrees=10
  twistLimitDegrees=3
 }
 {bone1Name=dyn_coat_frnt_r_01, bone2Name=dyn_coat_frnt_r_parent, swing1LimitDegrees=10, swing2LimitDegrees=10, twistLimitDegrees=3}
 {bone1Name=dyn_coat_frnt_r_parent, bone2Name=dyn_coat_frnt_r_base, swing1LimitDegrees=10, swing2LimitDegrees=10, twistLimitDegrees=3}
 {bone1Name=dyn_coat_frnt_l_01, bone2Name=dyn_coat_frnt_l_parent, swing1LimitDegrees=10, swing2LimitDegrees=10, twistLimitDegrees=3}
 {bone1Name=dyn_coat_frnt_l_parent, bone2Name=dyn_coat_frnt_l_base, swing1LimitDegrees=10, swing2LimitDegrees=10, twistLimitDegrees=3}
 {bone1Name=dyn_coat_bk_parent, bone2Name=dyn_coat_bk_base, swing1LimitDegrees=10, swing2LimitDegrees=15, twistLimitDegrees=3}
 {bone1Name=dyn_coat_bk_01, bone2Name=dyn_coat_bk_parent, swing1LimitDegrees=10, swing2LimitDegrees=20, twistLimitDegrees=3}
]
