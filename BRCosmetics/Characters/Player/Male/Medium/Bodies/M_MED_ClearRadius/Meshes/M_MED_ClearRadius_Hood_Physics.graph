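# Physics asset for the hood of the ClearRadius medium male player body, covering its dynamic hood bones.
# A small sphere on dyn_hood_base anchors a serial chain of capsules on dyn_hood_parent, dyn_hood_01 and dyn_hood_02.
# The capsules share one length and vary in radius, widest at dyn_hood_01.
# Each link is constrained to the previous one with identical tight swing and even tighter twist limits, restricting how far the hood bends.
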PhysicsAsset /BRCosmetics/Characters/Player/Male/Medium/Bodies/M_MED_ClearRadius/Meshes/M_MED_ClearRadius_Hood_Physics
bodies {
 dyn_hood_01 [{shapeType=Capsule, length=1.7469, radius=3.2111, shapeName=dyn_hood_01_capsule, center=<2.6237 0 0>, rotation=<rot 90 0 0>}]
 dyn_hood_02 [{shapeType=Capsule, length=1.7469, radius=2.1301, shapeName=dyn_hood_02_capsule, center=<2.6237 0 0>, rotation=<rot 90 0 0>}]
 dyn_hood_base [{shapeType=Sphere, radius=0.5, shapeName=dyn_hood_base_sphere}]
 dyn_hood_parent [{shapeType=Capsule, length=1.7469, radius=2.5738, shapeName=dyn_hood_parent_capsule, center=<2.6237 0 0>, rotation=<rot 90 0 0>}]
}
constraints [
 {bone1Name=dyn_hood_parent, bone2Name=dyn_hood_base, swing1LimitDegrees=10, swing2LimitDegrees=10, twistLimitDegrees=3}
 {bone1Name=dyn_hood_01, bone2Name=dyn_hood_parent, swing1LimitDegrees=10, swing2LimitDegrees=10, twistLimitDegrees=3}
 {bone1Name=dyn_hood_02, bone2Name=dyn_hood_01, swing1LimitDegrees=10, swing2LimitDegrees=10, twistLimitDegrees=3}
]
