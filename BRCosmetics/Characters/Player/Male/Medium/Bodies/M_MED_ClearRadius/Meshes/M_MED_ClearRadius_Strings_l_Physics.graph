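# Physics asset for the left-side strings on the male medium ClearRadius character body cosmetic.
# It covers two bone chains, inner and outer, each running base, parent, 01, 02, with a small sphere on the base and capsules on the other bones.
# Constraints link each bone to the previous one in its chain, with narrow swing and very tight twist limits, slightly stiffer at the root joint.
# No constraint connects the two chains.
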PhysicsAsset /BRCosmetics/Characters/Player/Male/Medium/Bodies/M_MED_ClearRadius/Meshes/M_MED_ClearRadius_Strings_l_Physics
bodies {
 dyn_string_inner_l_01 [{shapeType=Capsule, length=2.0733, radius=1.0475, shapeName=dyn_string_inner_l_01_capsule, center=<1.9023 0 0>, rotation=<rot 90 0 0>}]
 dyn_string_inner_l_02 [{shapeType=Capsule, length=1.2212, radius=1.0743, shapeName=dyn_string_inner_l_02_capsule, center=<1.5883 0 0>, rotation=<rot 90 0 0>}]
 dyn_string_inner_l_base [{shapeType=Sphere, radius=0.5, shapeName=dyn_string_inner_l_base_sphere}]
 dyn_string_inner_l_parent [{shapeType=Capsule, length=1.205, radius=1.0475, shapeName=dyn_string_inner_l_parent_capsule, center=<1.228 0 0>, rotation=<rot 90 0 0>}]
 dyn_string_outter_l_01 [{shapeType=Capsule, length=2.0782, radius=1.0328, shapeName=dyn_string_outter_l_01_capsule, center=<1.8835 0 0>, rotation=<rot 90 0 0>}]
 dyn_string_outter_l_02 [{shapeType=Capsule, length=1.3989, radius=1.2917, shapeName=dyn_string_outter_l_02_capsule, center=<1.8224 0 0>, rotation=<rot 90 0 0>}]
 dyn_string_outter_l_base [{shapeType=Sphere, radius=0.5, shapeName=dyn_string_outter_l_base_sphere}]
 dyn_string_outter_l_parent [{shapeType=Capsule, length=1.205, radius=1.0475, shapeName=dyn_string_outter_l_parent_capsule, center=<1.228 0 0>, rotation=<rot 90 0 0>}]
}
constraints [
 {bone1Name=dyn_string_inner_l_parent, bone2Name=dyn_string_inner_l_base, swing1LimitDegrees=10, swing2LimitDegrees=15, twistLimitDegrees=3}
 {bone1Name=dyn_string_inner_l_01, bone2Name=dyn_string_inner_l_parent, swing1LimitDegrees=15, swing2LimitDegrees=20, twistLimitDegrees=3}
 {bone1Name=dyn_string_inner_l_02, bone2Name=dyn_string_inner_l_01, swing1LimitDegrees=15, swing2LimitDegrees=20, twistLimitDegrees=3}
 {bone1Name=dyn_string_outter_l_parent, bone2Name=dyn_string_outter_l_base, swing1LimitDegrees=10, swing2LimitDegrees=15, twistLimitDegrees=3}
 {bone1Name=dyn_string_outter_l_01, bone2Name=dyn_string_outter_l_parent, swing1LimitDegrees=15, swing2LimitDegrees=20, twistLimitDegrees=3}
 {bone1Name=dyn_string_outter_l_02, bone2Name=dyn_string_outter_l_01, swing1LimitDegrees=15, swing2LimitDegrees=20, twistLimitDegrees=3}
]
