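# Physics asset for the right-side dangling strings on the male medium ClearRadius player cosmetic body, defining their simulated bodies and joints.
# Inner and outer strings form two separate four-bone chains, each running from a small base sphere through parent, 01 and 02 capsules.
# Constraints link each bone to the previous one with narrow twist and modest swing limits, tightest at the root joint.
# No constraint connects the two chains.
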PhysicsAsset /BRCosmetics/Characters/Player/Male/Medium/Bodies/M_MED_ClearRadius/Meshes/M_MED_ClearRadius_Strings_r_Physics
bodies {
 dyn_string_inner_r_01 [{shapeType=Capsule, length=1.5902, radius=1.4459, shapeName=dyn_string_inner_r_01_capsule, center=<1.7749 0 0>, rotation=<rot 90 0 0>}]
 dyn_string_inner_r_02 [{shapeType=Capsule, length=1.9546, radius=1.2219, shapeName=dyn_string_inner_r_02_capsule, center=<2.0062 0 0>, rotation=<rot 90 0 0>}]
 dyn_string_inner_r_base [{shapeType=Sphere, radius=0.5, shapeName=dyn_string_inner_r_base_sphere}]
 dyn_string_inner_r_parent [{shapeType=Capsule, length=1.205, radius=1.0475, shapeName=dyn_string_inner_r_parent_capsule, center=<1.228 0 0>, rotation=<rot 90 0 0>}]
 dyn_string_outter_r_01 [{shapeType=Capsule, length=2.0961, radius=1.0475, shapeName=dyn_string_outter_r_01_capsule, center=<1.9682 0 0>, rotation=<rot 90 0 0>}]
 dyn_string_outter_r_02 [{shapeType=Capsule, length=1.1021, radius=1.2249, shapeName=dyn_string_outter_r_02_capsule, center=<1.7192 0 0>, rotation=<rot 90 0 0>}]
 dyn_string_outter_r_base [{shapeType=Sphere, radius=0.5, shapeName=dyn_string_outter_r_base_sphere}]
 dyn_string_outter_r_parent [{shapeType=Capsule, length=1.205, radius=1.0475, shapeName=dyn_string_outter_r_parent_capsule, center=<1.228 0 0>, rotation=<rot 90 0 0>}]
}
constraints [
 {bone1Name=dyn_string_inner_r_parent, bone2Name=dyn_string_inner_r_base, swing1LimitDegrees=10, swing2LimitDegrees=15, twistLimitDegrees=3}
 {bone1Name=dyn_string_inner_r_01, bone2Name=dyn_string_inner_r_parent, swing1LimitDegrees=15, swing2LimitDegrees=20, twistLimitDegrees=3}
 {bone1Name=dyn_string_inner_r_02, bone2Name=dyn_string_inner_r_01, swing1LimitDegrees=15, swing2LimitDegrees=20, twistLimitDegrees=3}
 {bone1Name=dyn_string_outter_r_parent, bone2Name=dyn_string_outter_r_base, swing1LimitDegrees=10, swing2LimitDegrees=15, twistLimitDegrees=3}
 {bone1Name=dyn_string_outter_r_01, bone2Name=dyn_string_outter_r_parent, swing1LimitDegrees=15, swing2LimitDegrees=20, twistLimitDegrees=3}
 {bone1Name=dyn_string_outter_r_02, bone2Name=dyn_string_outter_r_01, swing1LimitDegrees=15, swing2LimitDegrees=20, twistLimitDegrees=3}
]
