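# Physics asset of collision shapes for the ClearRadius medium male player body; the name indicates it serves as colliders for cape simulation.
# Capsules cover the pelvis, spine chain, clavicles and neck, with a sphere on the head, and extra overlapping capsules on spine_02 and spine_05.
# Arm, leg and foot bodies exist but hold no shapes.
# It has no constraints, so it defines collision volumes only, with no jointed ragdoll setup.
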PhysicsAsset /BRCosmetics/Characters/Player/Male/Medium/Bodies/M_MED_ClearRadius/Meshes/M_MED_Clear_Radius_CapeColliders
bodies {
 calf_l []
 calf_r []
 clavicle_l [{shapeType=Capsule, length=7.7636, radius=11.1427, center=<9.7971 1.4456 -1.5504>, rotation=<rot 75.9352 -0.0001 175.5622>}]
 clavicle_r [{shapeType=Capsule, length=8.561, radius=10.3453, center=<-9.7668 -1.848 1.5429>, rotation=<rot 75.9351 -0.0001 -4.4378>}]
 foot_l []
 foot_r []
 head [{shapeType=Sphere, radius=17.3125, center=<11.028 -0.174 0>}]
 lowerarm_l []
 lowerarm_r []
 neck_02 [{shapeType=Capsule, length=13.4375, radius=7.3175, center=<-0.2642 -0.3408 0>, rotation=<rot 90 -171.1193 -175.3978>}]
 pelvis [{shapeType=Capsule, length=8.6562, radius=13.5388, center=<-4.1588 1.3718 0>}]
 spine_01 [{shapeType=Capsule, length=10.7187, radius=12.4138, center=<-4.1588 1.7372 0>}]
 spine_02 [
  {shapeType=Capsule, length=10.5937, radius=10.945, center=<-2.6215 1.7526 0.0001>}
  {shapeType=Capsule, length=29.875, radius=11.1013, center=<2.5024 1.7598 3.9825>, rotation=<rot 90 0.0099 0.0099>}
  {shapeType=Capsule, length=29.875, radius=11.1013, center=<2.4704 1.4503 -3.9825>, rotation=<rot 90 -179.9963 -179.9963>}
 ]
 spine_03 [{shapeType=Capsule, length=9.1562, radius=10.7575, center=<-1.953 1.4888 0.0001>}]
 spine_04 [{shapeType=Capsule, length=16.8702, radius=11.4186, center=<2.1467 3.1738 0.0002>}]
 spine_05 [
  {shapeType=Capsule, length=20.0576, radius=14.9811, center=<-1.7514 6.3661 0>}
  {shapeType=Capsule, length=12.7812, radius=17.0625, center=<0.7875 8.1512 0>}
 ]
 thigh_l []
 thigh_r []
 upperarm_l []
 upperarm_r []
}
constraints []
