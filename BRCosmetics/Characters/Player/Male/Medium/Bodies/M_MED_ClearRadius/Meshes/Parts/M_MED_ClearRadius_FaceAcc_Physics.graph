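# Physics asset for the face accessory mesh of the male medium ClearRadius player cosmetic, covering nine dynamic hair chains (03 and 05–12).
# Each chain pairs a small sphere on its base bone with a fitted capsule on the bone named parent, joined by a tightly limited swing and twist constraint.
# Chain 06 has a narrower limit on one swing axis.
# An offset sphere has no constraint, and no constraints link separate chains.
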PhysicsAsset /BRCosmetics/Characters/Player/Male/Medium/Bodies/M_MED_ClearRadius/Meshes/Parts/M_MED_ClearRadius_FaceAcc_Physics
bodies {
 dyn_hair_03_base [{shapeType=Sphere, radius=0.5, shapeName=dyn_hair_03_base_sphere}]
 dyn_hair_03_parent [{
   shapeType=Capsule
   length=4.0957
   radius=0.8515
   shapeName=dyn_hair_03_parent_capsule
   center=<1.8814 0 0.2329>
   rotation=<rot 71.524 179.9999 179.9999>
  }]
 dyn_hair_05_base [{shapeType=Sphere, radius=0.5, shapeName=dyn_hair_05_base_sphere}]
 dyn_hair_05_parent [{shapeType=Capsule, length=3.2119, radius=1.1966, shapeName=dyn_hair_05_parent_capsule, center=<1.5207 0 0>, rotation=<rot 90 0 0>}]
 dyn_hair_06_base [{shapeType=Sphere, radius=0.5, shapeName=dyn_hair_06_base_sphere}]
 dyn_hair_06_parent [{shapeType=Capsule, length=2.7491, radius=1.075, shapeName=dyn_hair_06_parent_capsule, center=<1.6184 0 0>, rotation=<rot 90 0 0>}]
 dyn_hair_07_base [{shapeType=Sphere, radius=0.5, shapeName=dyn_hair_07_base_sphere}]
 dyn_hair_07_parent [{shapeType=Capsule, length=3.0968, radius=1.4675, shapeName=dyn_hair_07_parent_capsule, center=<2.6412 0 0>, rotation=<rot 90 0 0>}]
 dyn_hair_08_base [{shapeType=Sphere, radius=0.5, shapeName=dyn_hair_08_base_sphere}]
 dyn_hair_08_parent [{shapeType=Capsule, length=2.6231, radius=0.9582, shapeName=dyn_hair_08_parent_capsule, center=<1.5213 0 0>, rotation=<rot 90 0 0>}]
 dyn_hair_09_base [{shapeType=Sphere, radius=0.5, shapeName=dyn_hair_09_base_sphere}]
 dyn_hair_09_parent [{shapeType=Capsule, length=2.325, radius=0.8978, shapeName=dyn_hair_09_parent_capsule, center=<1.2287 0 0>, rotation=<rot 90 0 0>}]
 dyn_hair_10_base [{shapeType=Sphere, radius=0.5, shapeName=dyn_hair_10_base_sphere}]
 dyn_hair_10_parent [{shapeType=Capsule, length=2.6188, radius=0.7795, shapeName=dyn_hair_10_parent_capsule, center=<1.5828 0 0>, rotation=<rot 90 0 0>}]
 dyn_hair_11_base [{shapeType=Sphere, radius=0.5, shapeName=dyn_hair_11_base_sphere}]
 dyn_hair_11_parent [{shapeType=Capsule, length=2.6947, radius=1.0211, shapeName=dyn_hair_11_parent_capsule, center=<1.5174 0 0>, rotation=<rot 90 0 0>}]
 dyn_hair_12_base [{shapeType=Sphere, radius=0.5, shapeName=dyn_hair_12_base_sphere}]
 dyn_hair_12_parent [{shapeType=Capsule, length=1.6541, radius=1.1774, shapeName=dyn_hair_12_parent_capsule, center=<1.3784 0 0>, rotation=<rot 90 0 0>}]
 dyn_hair_offset [{shapeType=Sphere, radius=0.5, shapeName=dyn_hair_offset_sphere}]
}
constraints [
 {bone1Name=dyn_hair_12_parent, bone2Name=dyn_hair_12_base, swing1LimitDegrees=10, swing2LimitDegrees=10, twistLimitDegrees=3}
 {bone1Name=dyn_hair_11_parent, bone2Name=dyn_hair_11_base, swing1LimitDegrees=10, swing2LimitDegrees=10, twistLimitDegrees=3}
 {bone1Name=dyn_hair_10_parent, bone2Name=dyn_hair_10_base, swing1LimitDegrees=10, swing2LimitDegrees=10, twistLimitDegrees=3}
 {bone1Name=dyn_hair_09_parent, bone2Name=dyn_hair_09_base, swing1LimitDegrees=10, swing2LimitDegrees=10, twistLimitDegrees=3}
 {bone1Name=dyn_hair_08_parent, bone2Name=dyn_hair_08_base, swing1LimitDegrees=10, swing2LimitDegrees=10, twistLimitDegrees=3}
 {bone1Name=dyn_hair_07_parent, bone2Name=dyn_hair_07_base, swing1LimitDegrees=10, swing2LimitDegrees=10, twistLimitDegrees=3}
 {bone1Name=dyn_hair_06_parent, bone2Name=dyn_hair_06_base, swing1LimitDegrees=5, swing2LimitDegrees=10, twistLimitDegrees=3}
 {bone1Name=dyn_hair_05_parent, bone2Name=dyn_hair_05_base, swing1LimitDegrees=10, swing2LimitDegrees=10, twistLimitDegrees=3}
 {bone1Name=dyn_hair_03_parent, bone2Name=dyn_hair_03_base, swing1LimitDegrees=10, swing2LimitDegrees=10, twistLimitDegrees=3}
]
